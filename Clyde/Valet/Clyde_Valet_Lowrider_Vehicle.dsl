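; Lowrider variant of the Clyde Valet vehicle that forces the lowrider vehicle mod on at spawn.
; BeginPlay runs the parent logic, removes any applied tire mod, then applies the mod from the Lowrider Mod Tag variable if CanApplyMod passes.
; The handbrake, enter/exit, jump, hydraulic audio, sparks, bounce-to-music and radio events are empty here, and the construction script only calls its parent.
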(event EventBeginPlay
  (bind _lowrider_mod_tag (Variables|Default|GetLowrider%20Mod%20Tag))
  (|CallParentFunction)
  ;[0.15 0 0.1418] Forcing the lowrider vehicle mod to always be applied to this vehicle at the start
  (if (Vehicle|Mod|Tire|HasTireMod)
    (Vehicle|Mod|Tire|RemoveAppliedTireMod)
    (else
      _))
  (if (Vehicle|Mod|CanApplyMod _lowrider_mod_tag)
    (Vehicle|Mod|ApplyVehicleMod _lowrider_mod_tag)))

(event EventOnStopHandbrake)

(event Custom|Reset%20Jump)

(event Custom|Disable%20Dropped)

(event EventOnStartHandbrake)

(event Custom|Lowrider%20Jump)

(event Custom|Hydraulic%20Audio%202)

(event Custom|Start%20Sparks)

(event Custom|BounceToMusic)

(event EventOnPawnEnterVehicle
  (params (PlayerPawn :type Fort%20Player%20Pawn%20Object%20Reference) (SeatIdx :type Integer)))

(event Custom|Sparking)

(event Custom|LoopingTrigger)

(event Custom|StartRadio
  (params (Component :type Streaming%20Radio%20Player%20Component%20Object%20Reference) (AffectedSource :type Athena%20Radio%20Station%20Structure)))

(event Custom|StopRadio
  (params (Component :type Streaming%20Radio%20Player%20Component%20Object%20Reference) (AffectedSource :type Athena%20Radio%20Station%20Structure)))

(event EventOnPawnExitVehicle
  (params (PlayerPawn :type Fort%20Player%20Pawn%20Object%20Reference) (ExitSocketName :type Name :default "None")))

(event Custom|Tilt%20Audio)

(event Custom|Fix%20Cosmetic%20Anim)

(event Custom|Tilt%20Audio%202)

(fn ConstructionScript
  (|CallParentFunction))
