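; Valet sports car vehicle Blueprint that ensures it carries the lowrider hydraulic mod when play begins.
; After the parent BeginPlay, it removes any applied tire mod, as the hydraulic mod does not count as one.
; It then applies the mod by tag if absent, covering cases where the hydraulic vehicle spawner is not used.
; Overlap and Tick events are empty, and construction only calls the parent.
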(event EventBeginPlay
  (bind _lowrider_mod_tag (Variables|Default|GetLowrider%20Mod%20Tag))
  (|CallParentFunction)
  (Utilities|FlowControl|Sequence
    (:then_0
      ;[0.15 0 0.1418] Removing any alternate tire mods because the hydraulic mod doesn't count as one
      (if (Vehicle|Mod|Tire|HasTireMod)
        (Vehicle|Mod|Tire|RemoveAppliedTireMod)))
    (:then_1
      ;[0.01568 0 0.1302] apply the hydraulic mod in case the hydraulic vehicle spawner is not used for this vehicle
      (if (not (Vehicle|Mod|HasVehicleModApplied _lowrider_mod_tag))
        (Vehicle|Mod|ApplyVehicleMod _lowrider_mod_tag)))))

(event Collision|EventActorBeginOverlap
  (params (OtherActor :type Actor%20Object%20Reference)))

(event EventTick
  (params (DeltaSeconds :type Float%20%28single-precision%29)))

(fn ConstructionScript
  (|CallParentFunction))
